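; Radar tower building Blueprint, set invulnerable at BeginPlay, that blinks a looping Morse code message on its light.
; Without network authority, it steps through the Code array using timers, mapping entries to short or long beeps and letter or word gaps.
; Set Light toggles the mesh's dynamic material emissive fresnel exponent, with an off pause after each beep.
; Custom gm_morse_* events retune the timing for debugging.
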(event Custom|Set%20Light
  (params (On :type Boolean))
  ; Morse Code Light
  ;; Light On / Off
  (Rendering|Material|SetScalarParameterValue (Class|BTechArtGlobalLibrary|GetDynamicMaterialInstance (Variables|Debug|GetStaticMeshComponent)) "EmissveFresnelExponent" (Math|Float|SelectFloat 0.5 0.0 On) :!class /Script/Engine.MaterialInstanceDynamic))

(event Custom|gm_morse_letter
  (params (Delay%20Between%20Letters :type Float%20%28double-precision%29))
  ; Morse Code Light
  ;; Debug Tuning
  (Variables|Default|SetDelay%20Between%20Letters Delay%20Between%20Letters))

(event Custom|gm_morse_short
  (params (Short%20Beep%20Length :type Float%20%28double-precision%29))
  ; Morse Code Light
  ;; Debug Tuning
  (Variables|Default|SetShort%20Beep%20Length Short%20Beep%20Length))

(event EventBeginPlay
  ; Morse Code Light
  ;; Start
  (|CallParentFunction)
  (Variables|BuildingActor|SetbIsInvulnerable true)
  (Networking|SwitchHasAuthority
    (:Remote
      (CallFunction|ChooseNextCode))))

(event Custom|gm_morse_beep
  (params (Delay%20Between%20Beeps :type Float%20%28double-precision%29))
  ; Morse Code Light
  ;; Debug Tuning
  (Variables|Default|SetDelay%20Between%20Beeps Delay%20Between%20Beeps))

(event Custom|gm_morse_word
  (params (Delay%20Between%20Words :type Float%20%28double-precision%29))
  ; Morse Code Light
  ;; Debug Tuning
  (Variables|Default|SetDelay%20Between%20Words Delay%20Between%20Words))

(event Custom|gm_morse_long
  (params (Long%20Beep%20Length :type Float%20%28double-precision%29))
  ; Morse Code Light
  ;; Debug Tuning
  (Variables|Default|SetLong%20Beep%20Length Long%20Beep%20Length))

(event Custom|Choose%20Next%20Code
  ; Morse Code Light
  ;; Start
  (Math|Integer|IncrementInt (Variables|Default|GetCode%20Index)
    (:%20%20
      (if (> (Variables|Default|GetCode%20Index) (Utilities|Array|LastIndex (Variables|Default|GetCode)))
        (Variables|Default|SetCode%20Index 0) ; reset back to the beginning
      )
      (CallFunction|SetLight (Utilities|Select :Index (Utilities|Array|Get(aref) (Variables|Default|GetCode) (Variables|Default|GetCode%20Index)) :Option%200 true :Option%201 true :Option%202 false :Option%203 false))
      (Utilities|Time|SetTimerbyEvent (delegate Custom|Finished%20Showing%20Code) (Utilities|Select :Index (Utilities|Array|Get(aref) (Variables|Default|GetCode) (Variables|Default|GetCode%20Index)) :Option%200 (Variables|Default|GetShort%20Beep%20Length) :Option%201 (Variables|Default|GetLong%20Beep%20Length) :Option%202 (Variables|Default|GetDelay%20Between%20Letters) :Option%203 (Variables|Default|GetDelay%20Between%20Words))))) ; show the next code
)

(event Custom|Finished%20Pause%20After%20Turning%20Light%20Off
  ; Morse Code Light
  ;; Start
  (CallFunction|ChooseNextCode) ; Continue to the next code after\nturning the light off from a short\nor long beep
)

(event Custom|Finished%20Showing%20Code
  ; Morse Code Light
  ;; Start
  (if (<= (Utilities|Array|Get(aref) (Variables|Default|GetCode) (Variables|Default|GetCode%20Index)) 1)
    (CallFunction|SetLight false) ; turn the light off after a beep
    (Utilities|Time|SetTimerbyEvent (delegate Custom|Finished%20Pause%20After%20Turning%20Light%20Off) (Variables|Default|GetDelay%20Between%20Beeps)) ; delay when turning the light off before\ndoing the next part of the code
    (else
      (CallFunction|ChooseNextCode) ; Continue to the next code if we just\nfinished a non-beep pause such\nas between letters or words
    )))

(fn ConstructionScript)
